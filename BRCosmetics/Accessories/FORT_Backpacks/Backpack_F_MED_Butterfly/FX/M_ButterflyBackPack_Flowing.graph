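# Additive, unlit, two-sided Niagara mesh-particle material that draws flowing energy for the butterfly backpack's wings.
# Two panning caustic textures, on UVs tapered by a curve atlas and driven by dynamic parameters, multiply to blend dark and highlight colors for emissive.
# Sphere masks, atlas curves, sine edge falloff and particle alpha shape opacity, softened by depth, near-camera and view-angle fades.
# A camera-proximity test yields an offset halfway toward the particle center.
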Material /BRCosmetics/Accessories/FORT_Backpacks/Backpack_F_MED_Butterfly/FX/M_ButterflyBackPack_Flowing
graph {
 properties {
  bUsedWithNiagaraMeshParticles=true
  blendMode=BLEND_Additive
  pixelDepthOffsetMode=PDOM_Legacy
  shadingModel=MSM_Unlit
  twoSided=true
  usedShadingModels=Unlit
 }
 nodes [
  n1 = ObjectPositionWS()
  n2 = CameraPositionWS()
  n3 = Subtract(A=ShadingPathSwitch(Default=ParticlePositionWS(), Mobile=ObjectPositionWS()) {desc=Migrated from FeatureLevelSwitch}, B=WorldPosition().XYZ)
  n4 = MaterialFunctionCall(Vector 3=@n3) {materialFunction=/Engine/Functions/Engine_MaterialFunctions02/Utility/VectorLength}
  If(
   A=MaterialFunctionCall(Vector 3=Subtract(A=@n1, B=@n2)) {materialFunction=&VectorLength}.V3 Length
   B=Constant() {r=40}
   "A > B"=Constant()
   "A < B"=Multiply(A=Normalize(VectorInput=@n3), B=Multiply(A=@n4.V3 Length) {constB=0.5})
  )
  n5 = ComponentMask(None=VectorParameter() {group=Texture1, parameterName=Texture1_UVtile, sortPriority=2, defaultValue=<rgba 0.4 6 0 1>}.RGB) {g=true, r=true}
  n6 = DynamicParameter() {paramNames ["U_offset(0-1)", "V_offset(0-1)", "U_tile(default-1)", "V_tile(default-1)"]}
  n7 = AppendVector(A=@n6."U_tile(default-1)", B=@n6."U_tile(default-1)")
  n8 = OneMinus(None=ComponentMask(None=TextureCoordinate()) {r=true})
  n9 = DynamicParameter() {paramNames ["U_offset(0-1)", "V_offset(0-1)", "U_tile(default-1)", "V_tile(default-1)"]}
  n10 = Add(A=@n9."V_offset(0-1)") {constB=0.4}
  n11 = Add(A=@n9."V_offset(0-1)") {constA=0.6, constB=0.6}
  n12 = TextureCoordinate()
  n13 = ComponentMask(None=@n12) {g=true}
  n14 = CurveAtlasRowParameter(CurveTime=ComponentMask(None=OneMinus(None=@n12)) {r=true}) {
   atlas=&AthenaCurveAtlas01
   curve=/Game/Accessories/FORT_Backpacks/Backpack_F_MED_Butterfly/FX/Curve_Butterfly_BackPack
   defaultValue=102
   group=Atlas_Curve
   parameterName="TaperAjustCurve(0in-1out)"
  }
  n15 = AppendVector(
   A=Multiply(A=Add(A=@n8, B=@n9."U_offset(0-1)"))
   B=LinearInterpolate(
    A=LinearInterpolate(A=@n10, B=@n11, Alpha=@n13) {constA=0.4, constB=0.6}
    B=Add(A=@n9."V_offset(0-1)", B=@n13)
    Alpha=StaticComponentMaskParameter(None=@n14) {defaultG=true, group=Atlas_Curve, parameterName=Taper_Channel, sortPriority=1}
   )
  )
  n16 = Panner(
   Coordinate=Multiply(A=@n15, B=Multiply(A=@n5, B=@n7))
   Speed=ComponentMask(None=VectorParameter() {group=Texture1, parameterName=Texture1_speed, sortPriority=3, defaultValue=<rgba 0.1 0.02 0 1>}.RGB) {g=true, r=true}
  ) {speedX=0.1, speedY=0.02}
  n17 = Multiply(A=StaticComponentMaskParameter(None=TextureSampleParameter2D(UVs=@n16) {group=Texture1, parameterName=FlowingTexture1, texture=/Game/Global/Textures/Default/Caustics/T_Caustics_Softer}.RGB) {defaultR=true, group=Texture1, parameterName=Texture1mask, sortPriority=1}) {constB=1.5}
  n18 = ComponentMask(None=VectorParameter() {group=Texture2, parameterName=Texture2_UVtile, sortPriority=2, defaultValue=<rgba 1.5 15 0 1>}.RGB) {g=true, r=true}
  n19 = Panner(
   Coordinate=Multiply(A=@n15, B=Multiply(A=@n7, B=@n18))
   Speed=ComponentMask(None=VectorParameter() {group=Texture2, parameterName=Texture2_speed, sortPriority=3, defaultValue=<rgba 0.15 -0.02 0 1>}.RGB) {g=true, r=true}
  ) {speedX=0.15, speedY=-0.02}
  n20 = Multiply(A=StaticComponentMaskParameter(None=TextureSampleParameter2D(UVs=@n19) {group=Texture2, parameterName=FlowingTexture2, texture=&T_ShieldCaustics}.RGB) {defaultR=true, group=Texture2, parameterName=Texture2Mask, sortPriority=1}) {constB=1.4}
  MaterialFunctionCall(Emissive=LinearInterpolate(
    A=VectorParameter() {parameterName=DarkColor, defaultValue=<rgba 0 0 0 1>}.RGB
    B=VectorParameter() {parameterName=HightLightColor, defaultValue=<rgba 0.2542 0.7157 1 1>}.RGB
    Alpha=Multiply(A=@n17, B=@n20)
   )) {materialFunction=&MF_FX_EmissiveAdjustments}
  n21 = TextureCoordinate()
  n22 = OneMinus(None=@n21)
  n23 = AppendVector(
   A=LinearInterpolate(Alpha=ComponentMask(None=@n22) {r=true}) {constB=0.55}
   B=LinearInterpolate(Alpha=ComponentMask(None=@n21) {g=true}) {constB=0.85}
  )
  n24 = Constant2Vector() {g=0.5}
  n25 = Constant() {r=0.5}
  n26 = Constant()
  n27 = TextureCoordinate()
  n28 = ComponentMask(None=OneMinus(None=@n27)) {r=true}
  n29 = ComponentMask(None=@n27) {g=true}
  n30 = SphereMask(
   A=AppendVector(A=LinearInterpolate(Alpha=@n28) {constB=0.55}, B=LinearInterpolate(Alpha=@n29) {constB=1.15})
   B=Constant2Vector() {g=0.8}
   Radius=Constant() {r=0.8}
   Hardness=Constant() {r=0.3}
  )
  n31 = CurveAtlasRowParameter(CurveTime=Multiply(A=SphereMask(A=@n23, B=@n24, Radius=@n25, Hardness=@n26), B=OneMinus(None=@n30))) {
   atlas=&AthenaCurveAtlas01
   curve=&Curve_Butterfly_BackPack
   defaultValue=102
   group=Atlas_Curve
   parameterName=WingShapeMaskIntensity_tweak
   sortPriority=4
  }
  n32 = StaticComponentMaskParameter(None=@n31) {defaultR=true, group=Atlas_Curve, parameterName=WingShapeMask_Channel, sortPriority=5}
  n33 = CurveAtlasRowParameter(CurveTime=ComponentMask(None=TextureCoordinate()) {g=true}) {
   atlas=&AthenaCurveAtlas01
   curve=&Curve_Butterfly_BackPack
   defaultValue=102
   group=Atlas_Curve
   parameterName=WingShapeDownPartAdjust
   sortPriority=4
  }
  n34 = CurveAtlasRowParameter(CurveTime=Multiply(A=@n32, B=@n33.A)) {
   atlas=&AthenaCurveAtlas01
   curve=&Curve_Butterfly_BackPack
   defaultValue=102
   group=Atlas_Curve
   parameterName=WingShapeDownPartAdjust
   sortPriority=4
  }
  n35 = Add(A=@n32, B=@n34.R)
  n36 = CurveAtlasRowParameter(CurveTime=ComponentMask(None=TextureCoordinate()) {r=true}) {
   atlas=&AthenaCurveAtlas01
   curve=&Curve_Butterfly_BackPack
   defaultValue=102
   group=Atlas_Curve
   parameterName="GradiantOpacity(U)"
   sortPriority=2
  }
  n37 = StaticComponentMaskParameter(None=@n36) {defaultB=true, group=Atlas_Curve, parameterName=Gradiant_channel, sortPriority=3}
  n38 = Sine(None=Divide(A=TextureCoordinate()))
  n39 = ComponentMask(None=@n38) {r=true}
  n40 = ComponentMask(None=@n38) {g=true}
  n41 = Multiply(A=ParticleColor().A, B=Multiply(A=Multiply(A=@n35, B=@n37), B=Multiply(A=@n39, B=@n40)))
  n42 = DepthFade(Opacity=@n41, FadeDistance=ScalarParameter() {defaultValue=10, parameterName=DepthDistance})
  n43 = Constant() {r=100}
  n44 = CameraVectorWS()
  n45 = VertexNormalWS()
  n46 = CurveAtlasRowParameter(CurveTime=Abs(None=DotProduct(A=@n44, B=@n45))) {
   atlas=&AthenaCurveAtlas01
   curve=/Game/Accessories/FORT_Backpacks/Backpack_F_MED_Butterfly/FX/Curve_Butterfly_BackPack1
   defaultValue=103
   parameterName=SoftTheTexture
  }
  Multiply(
   A=Multiply(
    A=ShadingPathSwitch(Default=@n42, Mobile=@n41) {desc=Migrated from FeatureLevelSwitch}
    B=MaterialFunctionCall(Distance=@n43) {materialFunction=&Near_Camera_Fade}.Result
   )
   B=@n46.R
  )
 ]
}
