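# Additive, unlit, two-sided material for Niagara mesh particles that draws flowing streaks on the Butterfly backpack, apparently for its side view.
# Particle color is masked by a panning caustics texture, with UVs mirrored around center, compressed and offset between halves, then fed to an emissive adjustment function.
# A soft mask multiplies curve-atlas fades along both UV axes, a sphere texture and a view-angle curve, then passes through depth fade.
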Material /BRCosmetics/Accessories/FORT_Backpacks/Backpack_F_MED_Butterfly/FX/M_ButterflyBackPack_sideViewFlowing
graph {
 properties {
  bUsedWithNiagaraMeshParticles=true
  blendMode=BLEND_Additive
  pixelDepthOffsetMode=PDOM_Legacy
  shadingModel=MSM_Unlit
  twoSided=true
  usedShadingModels=Unlit
 }
 nodes [
  n1 = TextureCoordinate()
  n2 = CurveAtlasRowParameter(CurveTime=ComponentMask(None=@n1) {r=true}) {
   atlas=&AthenaCurveAtlas01
   curve=/Game/Accessories/FORT_Backpacks/Backpack_F_MED_Butterfly/FX/Curve_Butterfly_BackPack1
   defaultValue=103
   parameterName=FadeCurve_horizontal
  }
  n3 = CurveAtlasRowParameter(CurveTime=ComponentMask(None=@n1) {g=true}) {
   atlas=&AthenaCurveAtlas01
   curve=&Curve_Butterfly_BackPack1
   defaultValue=103
   parameterName=FadeCurve_vertical
  }
  n4 = Multiply(A=@n2.G, B=@n3.B)
  n5 = TextureSample() {texture=/Game/Effects/Fort_Effects/Textures/Mask/T_Sphere_A}
  n6 = CameraVectorWS()
  n7 = VertexNormalWS()
  n8 = CurveAtlasRowParameter(CurveTime=Abs(None=DotProduct(A=@n6, B=@n7))) {
   atlas=&AthenaCurveAtlas01
   curve=&Curve_Butterfly_BackPack1
   defaultValue=103
   parameterName=SoftTheTexture
  }
  DepthFade(Opacity=Multiply(A=Multiply(A=@n4, B=@n5.R), B=@n8.A)) {fadeDistanceDefault=10}
  n9 = TextureCoordinate()
  n10 = ComponentMask(None=@n9) {r=true}
  n11 = Abs(None=Subtract(A=Multiply(A=@n10) {constB=2}))
  n12 = LinearInterpolate(Alpha=ComponentMask(None=@n9) {g=true}) {constB=0.2}
  n13 = Add(A=@n12) {constB=0.1}
  n14 = Constant()
  n15 = If(A=ComponentMask(None=@n9) {r=true}, B=Constant() {r=0.5}, "A > B"=Constant() {r=1}, "A == B"=@n14, "A < B"=@n14)
  n16 = Panner(
   Coordinate=AppendVector(A=LinearInterpolate(Alpha=@n11) {constB=0.1}, B=LinearInterpolate(A=@n12, B=@n13, Alpha=@n15))
   Speed=Constant2Vector() {r=-0.05}
  )
  MaterialFunctionCall(Emissive=LinearInterpolate(B=ParticleColor().RGB, Alpha=TextureSample(UVs=@n16) {texture=/Game/Global/Textures/Default/Caustics/T_Caustics_Softer}.R)) {materialFunction=&MF_FX_EmissiveAdjustments}
 ]
}
