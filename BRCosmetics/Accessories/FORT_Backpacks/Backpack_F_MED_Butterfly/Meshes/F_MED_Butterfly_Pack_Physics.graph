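# Physics asset for the F_MED_Butterfly backpack cosmetic, defining simple collision bodies and joints for its four wing bones.
# Small box bodies sit on the left and right, top and bottom wing bones, with a larger box on the root.
# Each wing is constrained directly to the root with identical swing limits and a tight twist limit, so the wings can swing separately with little twist.
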PhysicsAsset /BRCosmetics/Accessories/FORT_Backpacks/Backpack_F_MED_Butterfly/Meshes/F_MED_Butterfly_Pack_Physics
bodies {
 dyn_wing_l_bottom [{shapeType=Box, extentX=2.5, extentY=2.5, extentZ=2.5, center=<-0.9492 -0.0001 0.4778>, rotation=<rot 0 -90 0>}]
 dyn_wing_l_top [{shapeType=Box, extentX=2.5, extentY=2.5, extentZ=2.5, center=<-0.9492 -0.0001 0.4778>, rotation=<rot 0 -90 0>}]
 dyn_wing_r_bottom [{shapeType=Box, extentX=2.5, extentY=2.5, extentZ=2.5, center=<-0.9492 -0.0001 0.4778>, rotation=<rot 0 -90 0>}]
 dyn_wing_r_top [{shapeType=Box, extentX=2.5, extentY=2.5, extentZ=2.5, center=<-0.9492 -0.0001 0.4778>, rotation=<rot 0 -90 0>}]
 root [{shapeType=Box, extentX=4.5, extentY=5.5, extentZ=5.5}]
}
constraints [
 {bone1Name=dyn_wing_r_top, bone2Name=root, swing1LimitDegrees=45, swing2LimitDegrees=35, twistLimitDegrees=5}
 {bone1Name=dyn_wing_l_top, bone2Name=root, swing1LimitDegrees=45, swing2LimitDegrees=35, twistLimitDegrees=5}
 {bone1Name=dyn_wing_r_bottom, bone2Name=root, swing1LimitDegrees=45, swing2LimitDegrees=35, twistLimitDegrees=5}
 {bone1Name=dyn_wing_l_bottom, bone2Name=root, swing1LimitDegrees=45, swing2LimitDegrees=35, twistLimitDegrees=5}
]
